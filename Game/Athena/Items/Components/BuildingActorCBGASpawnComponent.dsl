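; Building actor component that drops consumable CBGA pickups when its owning building is damaged or destroyed.
; On the server, once playlist data is ready, scalable-float toggles gate binding to damage and death events.
; Damage checks consumable health thresholds, while death spawning is optional; both feed DynamicallySpawn.
; It throttles with DoOnce and Delay, async-loads the class once, rolls chance and quantity, then spawns around a rotated offset with optional impulse.
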(event EventBeginPlay
  ;[1 1 1] Begin Play - Wait for playlist then start spawning if enabled.
  (if (Networking|IsServer)
    (Class|FortAsyncActionCurrentPlaylistReady|PlaylistDataReadyAsync
      (:OnReady
        (if (and (Math|Conversions|ToBoolean(Integer) (Math|Float|Truncate (ScalableFloat|GetValueatLevel (Variables|Spawning|GetRow_GlobalEnabled)))) (Math|Conversions|ToBoolean(Integer) (Math|Float|Truncate (ScalableFloat|GetValueatLevel (Variables|Spawning|GetRow_SubtypeEnabled)))))
          (bind _asbuilding_actor (Utilities|Casting|CastToBuildingActor (Components|GetOwner)))
          (Variables|Spawning|SetOwningBuildingActor _asbuilding_actor)
          (EventDispatchers|BindEventtoOnDamaged _asbuilding_actor (EventDispatchers|CreateEvent :!event Custom|Damaged))
          (EventDispatchers|BindEventtoOnDied _asbuilding_actor (EventDispatchers|CreateEvent :!event Custom|Died)))))))

(event Custom|Damaged
  (params (DamagedActor :type Actor%20Object%20Reference) (Damage :type Float%20%28single-precision%29) (InstigatedBy :type Controller%20Object%20Reference) (DamageCauser :type Actor%20Object%20Reference) (HitLocation :type Vector) (FHitComponent :type Primitive%20Component%20Object%20Reference) (BoneName :type Name) (Momentum :type Vector))
  ;[1 1 1] On Damaged drop the the Dynamically Spawned CBGA
  (CallFunction|DynamicallySpawn true))

(event Custom|Died
  (params (DamagedActor :type Actor%20Object%20Reference) (Damage :type Float%20%28single-precision%29) (InstigatedBy :type Controller%20Object%20Reference) (DamageCauser :type Actor%20Object%20Reference) (HitLocation :type Vector) (FHitComponent :type Primitive%20Component%20Object%20Reference) (BoneName :type Name) (Momentum :type Vector))
  ;[1 1 1] On Damaged drop the the Dynamically Spawned CBGA
  (if (Variables|Spawning|GetRollSpawnOnDeath)
    (CallFunction|DynamicallySpawn false)))

(event Custom|DynamicallySpawn
  (params (UsesThresholdChecks :type Boolean))
  (bind _owningbuildingactor (Variables|Spawning|GetOwningBuildingActor))
  (bind _returnvalue (Transformation|GetActorLocation _owningbuildingactor))
  (bind _returnvalue_2 (Transformation|GetActorRotation _owningbuildingactor))
  (bind _relativespawnoffset (Variables|Spawning|GetRelativeSpawnOffset))
  (bind _returnvalue_3 (Math|Vector|RotateVector _relativespawnoffset _returnvalue_2))
  (bind _cbga_classref (Variables|Spawning|GetCBGA_ClassRef))
  ;[1 1 1] Dynamically Spawn CBGAs in
  ;;[1 1 1] Do a check to make sure we don't try to do this too often in too short of a time.
  (Utilities|FlowControl|DoOnce false
    (:Completed
      (Utilities|FlowControl|Sequence
        (:then_0
          _)
        (:then_1
          (Utilities|FlowControl|Delay (Variables|Spawning|GetDelayBetweenSpawnAttempts))))
      (Utilities|DoOnceThenDo false
        (:DoOnce
          ;;[1 1 1] Async load the actor then set the spawn class. only do this the first time through
          (bind _class (Utilities|AsyncLoadClassAsset (Variables|Spawning|GetCBGA_SoftRef))
            (:Completed
              (bind _ascbga_parent_physics (Utilities|Casting|CastToCBGA_Parent_PhysicsClass _class))
              (Variables|Spawning|SetCBGA_ClassRef _ascbga_parent_physics))))
        (:ThenDo
          (if (Utilities|IsValidClass (Variables|Spawning|GetCBGA_ClassRef))
            _)))
      (if UsesThresholdChecks
        ;;[1 1 1] Check if we hit a threshold then roll for a drop.
        (bind _validdrop (CallFunction|ThreshHoldCheck))
        (if _validdrop
          _)
        (else
          _))
      (if (Math|Random|RandomBoolwithWeight (ScalableFloat|GetValueatLevel (Variables|Spawning|GetRow_DropChance)))
        (for _ (range 1 (+ (Math|Random|RandomIntegerinRange 1 (Math|Float|Truncate (ScalableFloat|GetValueatLevel (Variables|Spawning|GetRow_MaxSpawnQuantityPerTrigger)))) 1))
          (if (Utilities|IsValidClass _cbga_classref)
            (bind _returnvalue_4 (Game|SpawnActorfromClass _cbga_classref (Math|Transform|MakeTransform (Math|Vector|vector+vector _returnvalue_3 _returnvalue (* (Math|Vector|RotateVectorAroundAxis "1.000000,0.000000,0.000000" (Math|Random|RandomFloatinRange (Variables|Spawning|GetSpawnRotationMin) (Variables|Spawning|GetSpawnRotationMax)) "0, 0,1.000000") (Variables|Spawning|GetSpawnXYForwardOffset))) (select (Variables|Spawning|GetShouldUseActorRotation) _returnvalue_2 (Variables|Spawning|GetSpecifiedSpawnRotation))) "AlwaysSpawn" "SelectDefaultAtRuntime" 0 true (Variables|Spawning|GetCollisionOnSpawn) 0 0))
            (Variables|/Game/Athena/BuildingActors/ConsumableBGAs/CBGA_Parent.CBGA_Parent_C|SetReplicatedCanInteract true _returnvalue_4)
            (if (Variables|ImpulseAfterSpawn|GetShouldImpulse)
              (Class|CBGAParentPhysics|Impulse _returnvalue_4 (Variables|ImpulseAfterSpawn|GetDelayBeforeImpulse) (Variables|ImpulseAfterSpawn|GetDelayBeforeCollisionAfterImpulse) (+ _returnvalue _returnvalue_3) (Variables|ImpulseAfterSpawn|GetImpulseAmount) (Variables|ImpulseAfterSpawn|GetCollisionAfterImpulse)))))))))

(fn ThreshHoldCheck
  (bind _thesholdamounts (Variables|Spawning|GetThesholdAmounts))
  (bind (_indexofmaxvalue _maxvalue) (Math|Float|MaxOfFloatArray _thesholdamounts))
  ;[1 1 1] Look for if a threshold was hit if so drop a CBGA
  (if (<= (Building|GetBuildingHealthPercentage (Variables|Spawning|GetOwningBuildingActor)) _maxvalue)
    (Utilities|Array|RemoveIndex (Variables|Spawning|GetThesholdAmounts) _indexofmaxvalue)
    (Development|PrintString "Threshold Hit")
    (return true)
    (else
      (return false))))
