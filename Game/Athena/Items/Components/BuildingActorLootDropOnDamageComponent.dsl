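; Building actor component that gives a chance to drop loot when its owner is damaged, with an optional Athena-only gate.
; BeginPlay binds the owner's OnDamaged, which uses a DoOnce and next-tick delay intended to count multi-hit shotgun damage once per tick.
; A scalable-float roll then tosses table loot from the hit location or bounds origin, offset relative to the owner.
; The tag-filtered gameplay event path is deprecated, so damage tags are temporarily ignored.
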(event EventBeginPlay
  ;[1 1 1] On Begin play setup
  (if (Variables|Default|GetEnabled)
    (if (and (FortGlobals|IsinAthena) (Variables|Default|GetAthenaOnly))
      (bind _asbuilding_actor (Utilities|Casting|CastToBuildingActor (Components|GetOwner)))
      (bind _output_get (Variables|Default|SetBuildingActorOwner _asbuilding_actor))
      (EventDispatchers|BindEventtoOnDamaged _output_get (EventDispatchers|CreateEvent :!event Custom|OnDamaged)))))

(event Custom|WaitDamageReceived
  ;[0 0 0] DEPRECATED Setup the wait for damage received event
  (Class|FortAsyncActionCurrentPlaylistReady|PlaylistDataReadyAsync
    (:OnReady
      (bind _buildingactorowner (Variables|Default|GetBuildingActorOwner)
        (:then
          (bind (_asynctaskproxy _payload) (Ability|Async|WaitGameplayEventToActor _buildingactorowner (Variables|Default|GetDamageReceivedEventTag) true)
            (:EventReceived
              (if (GameplayTags|DoesContainerMatchTagQuery (.InstigatorTags (Utilities|Struct|BreakGameplayEventData _payload)) (Variables|Default|GetEventDamageTagQuery))
                (CallFunction|DamageRecieved 0)))))
        (:else)))))

(event Custom|OnDamaged
  (params (DamagedActor :type Actor%20Object%20Reference) (Damage :type Float%20%28single-precision%29) (InstigatedBy :type Controller%20Object%20Reference) (DamageCauser :type Actor%20Object%20Reference) (HitLocation :type Vector) (FHitComponent :type Primitive%20Component%20Object%20Reference) (BoneName :type Name) (Momentum :type Vector))
  ;[0.6687 0.7139 1] On BuildingActorOwner Damaged
  ;
  ; TEMP for 22.00, allow ANY type of damage to trigger the loot.  For 22.00, this component's only reference is as a parent of BuildingActorLootDropComponent_ChromeProps which has EventDamageTagQuery set to ANY.
  ; For 22.10, will update OnDamaged event to pass the damage tags so EventDamageTagQuery can be used properly again.
  (Utilities|FlowControl|DoOnce false
    (:Completed
      (CallFunction|DamageRecieved DamageCauser HitLocation)
      (Utilities|FlowControl|DelayUntilNextTick))) ; Only process damage once per tick to handle \nshotguns doing multiple instances of damage
)

(event Custom|DamageRecieved
  (params (DamageCauser :type Actor%20Object%20Reference) (Hit%20Location :type Vector))
  ;[1 1 1] On damage roll to spawn loot
  (if (<= (Math|Random|RandomFloatinRange 0.0 1.0) (ScalableFloat|GetValueatLevel (Variables|Default|GetSF_ChanceToDropLoot)))
    (bind _buildingactorowner (Variables|Default|GetBuildingActorOwner)
      (:then
        (Class|AthenaFunctionLibrary|SpawnLootfromTable (Variables|Default|GetLootOnDamagePackage) 1 -1 (+ (Math|Vector|SelectVector (Collision|GetActorBounds _buildingactorowner true false) Hit%20Location (Utilities|Operators|Equal(==) :A Hit%20Location)) (CallFunction|MakeRelativeOffset _buildingactorowner (Variables|Default|GetOffsetLocation))) (Transformation|GetActorForwardVector _buildingactorowner) 0 true true false 0 "Tossed" "LootDrop"))
      (:else))))

(fn MakeRelativeOffset
  (params (Target :type Actor%20Object%20Reference) (InVec :type Vector))
  ;[1 1 1] Makes the offset relative to the target actor
  (return (Math|Vector|vector+vector (* (Transformation|GetActorForwardVector Target) (.x InVec)) (* (Transformation|GetActorRightVector Target) (.y InVec)) (* (Transformation|GetActorUpVector Target) (.z InVec)))))
