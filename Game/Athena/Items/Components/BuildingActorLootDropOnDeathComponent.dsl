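; Component that spawns loot when its owning building actor dies, active only when enabled, flagged Athena-only and running in Athena.
; BeginPlay binds the owner's OnDied event and caches the current playlist's max team size.
; On death, after an optional delay, it selects a loot tier mapped to team size or the default package.
; Loot spawns at an offset along the actor's local axes, with an optional hotfixable scalable-float override.
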(event EventBeginPlay
  ;[1 1 1] On Begin play bind to died if it is building actor
  (if (Variables|Default|GetEnabled)
    (if (and (FortGlobals|IsinAthena) (Variables|Default|GetAthenaOnly))
      (bind _asbuilding_actor (Utilities|Casting|CastToBuildingActor (Components|GetOwner))
        (:then
          (EventDispatchers|AssignOnDied _asbuilding_actor (delegate Custom|OnDied_Event))
          (Variables|Default|SetBuildingActorOwner _asbuilding_actor)
          (bind (_gamestate _playlist) (Class|FortAsyncActionCurrentPlaylistReady|PlaylistDataReadyAsync)
            (:OnReady
              (Variables|InternalData|SetCurrentPlaylistTeamSize (Variables|/Script/FortniteGame.FortPlaylist|GetMaxTeamSize _playlist)))))
        (:CastFailed
          (Development|PrintString "This Component is being used for Building actors, This is not a building actor" true true "(R=1.000000,G=0.728371,B=0.000000,A=1.000000)"))))))

(event Custom|OnDied_Event
  (params (DamagedActor :type Actor%20Object%20Reference) (Damage :type Float%20%28single-precision%29) (InstigatedBy :type Controller%20Object%20Reference) (DamageCauser :type Actor%20Object%20Reference) (HitLocation :type Vector) (FHitComponent :type Primitive%20Component%20Object%20Reference) (BoneName :type Name) (Momentum :type Vector))
  (bind _spawndelayafterdiedevent (Variables|Default|GetSpawnDelayAfterDiedEvent))
  (bind _returnvalue (ScalableFloat|GetValueatLevel _spawndelayafterdiedevent))
  (bind _currentplaylistteamsize (Variables|InternalData|GetCurrentPlaylistTeamSize))
  (bind _lootonexplodeperteamsize (Variables|Default|GetLootOnExplodePerTeamSize))
  (bind (_value _returnvalue_2) (Utilities|Map|Find _lootonexplodeperteamsize _currentplaylistteamsize))
  ;[1 1 1] On died spawn loot
  (Utilities|IsValid (Variables|Default|GetBuildingActorOwner)
    (:Is%20Valid
      (if (> _returnvalue 0)
        (Utilities|FlowControl|Delay _returnvalue)
        (else
          _))
      (if (Variables|Default|GetUsePlaylistTeamSizeForLootNames)
        (if _returnvalue_2
          (CallFunction|SpawnLootatLocation _value)
          (else
            _))
        (else
          _))
      (CallFunction|SpawnLootatLocation (Variables|Default|GetLootOnExplodePackage)))))

(event Custom|SpawnLootAtLocation
  (params (LootTierName :type Name))
  ; Spawn Loot
  (Class|AthenaFunctionLibrary|SpawnLootfromTable LootTierName 1 -1 (+ (Transformation|GetActorLocation (Variables|Default|GetBuildingActorOwner)) (CallFunction|MakeRelativeOffset (Variables|Default|GetBuildingActorOwner) (select (ScalableFloat|GetValueasBool (Variables|Default|LootOffset|GetUseHotfixedOffsetLocations)) (Math|Vector|MakeVector (ScalableFloat|GetValueatLevel (Variables|Default|LootOffset|GetHotfixedOffsetLocation_X)) (ScalableFloat|GetValueatLevel (Variables|Default|LootOffset|GetHotfixedOffsetLocation_Y)) (ScalableFloat|GetValueatLevel (Variables|Default|LootOffset|GetHotfixedOffsetLocation_Z))) (Variables|Default|LootOffset|GetOffsetLocation)))) (Transformation|GetActorForwardVector (Variables|Default|GetBuildingActorOwner)) 0 true true false 0 (Variables|Default|GetSpawnLootSourceType) (Variables|Default|GetSpawnLootSource)))

(fn MakeRelativeOffset
  (params (Target :type Actor%20Object%20Reference) (InVec :type Vector))
  ;[1 1 1] Makes the offset relative to the target actor
  (return (Math|Vector|vector+vector (* (Transformation|GetActorForwardVector Target) (.x InVec)) (* (Transformation|GetActorRightVector Target) (.y InVec)) (* (Transformation|GetActorUpVector Target) (.z InVec)))))
