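; Imminent destruction component that clears the path of its moving owner, such as a drop pod, by destroying buildings and hurting other actors.
; While enabled, Tick synchronously sphere-traces ahead along velocity so obstacles break before collision stops the owner.
; Eligible buildings die on the server and are death-predicted on clients, limited by tags, health and a count cap ignoring low-health pieces.
; Other actors receive server-applied damage and knockback GameplayEffects once per activation.
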(event Custom|Set%20Imminent%20Destruction%20Active
  (params (bActive :type Boolean))
  ;[0.0812 0.15 0.0546 0.5] Enable / Disable
  (Components|Tick|SetComponentTickEnabled bActive)
  (Utilities|Array|Clear (Variables|Runtime|GetActorsHit)))

(event EventTick
  (params (DeltaSeconds :type Float%20%28single-precision%29))
  (bind _returnvalue (Components|GetOwner))
  ;[0.03818 0.15 0.08373 0.5] Tick: Perform a sphere trace that covers the area the drop pod is about to travel. Deal damage to hits.
  ; This needs to be syncronous because otherwise the drop pod will collide and stop before the hits can be destroyed
  (|CallParentFunction DeltaSeconds)
  (Utilities|IsValid _returnvalue
    (:Is%20Valid
      (bind _outhits (Collision|MultiSphereTraceByProfile (Transformation|GetActorLocation _returnvalue) (+ (Transformation|GetActorLocation _returnvalue) (Utilities|Operators|Multiply (Transformation|GetVelocity _returnvalue) DeltaSeconds (ScalableFloat|GetValueatLevel (Variables|Hotfix|GetHotfix_TraceLengthMultiplier)))) (ScalableFloat|GetValueatLevel (Variables|Hotfix|GetHotfix_ImminentDestructionRadius)) "OverlapAll" (Variables|Default|GetbTraceComplex) (Utilities|Array|MakeArray _returnvalue) "None" true "(R=1.000000,G=0.000000,B=0.000000,A=1.000000)" "(R=0.000000,G=1.000000,B=0.000000,A=1.000000)" 2e+01))
      (for _array_element _outhits
        (CallFunction|DealDamagetoActor (.HitActor (Collision|BreakHitResult _array_element)) (.HitComponent (Collision|BreakHitResult _array_element)))))))

(event Custom|DealDamageToActor
  (params (HitActor :type Actor%20Object%20Reference) (HitComponent :type Primitive%20Component%20Object%20Reference))
  (bind _returnvalue (Components|GetOwner))
  (bind _instigator (Variables|/Script/Engine.Actor|GetInstigator _returnvalue))
  (bind _returnvalue_2 (Ability|GetFortAbilitySystemComponent _instigator))
  (bind _returnvalue_3 (Ability|GetFortAbilitySystemComponent HitActor))
  (bind _buseinstigatorabilitysystemcomponent (Variables|Default|GetbUseInstigatorAbilitySystemComponent))
  (bind _returnvalue_4 (Ability|GetFortAbilitySystemComponent _returnvalue))
  (bind _returnvalue_5 (select _buseinstigatorabilitysystemcomponent _returnvalue_2 _returnvalue_4))
  (bind _returnvalue_6 (GameplayEffects|MakeFortEffectContext _returnvalue_5 _returnvalue _returnvalue _returnvalue 0))
  ;[0.15 0.1028 0.04104 0.5] Deal Damage to Actor: Apply a damage GE to hit actor, or if it is a building destroy it up to the max buildings we'd like to destroy
  (bind _asbuilding_static_mesh (Utilities|Casting|CastToBuildingSMActor HitActor)
    (:then
      (if (CallFunction|ShouldDestroyBuilding _asbuilding_static_mesh)
        (CallFunction|DestroyBuilding _asbuilding_static_mesh HitComponent)))
    (:CastFailed
      (Utilities|IsValid HitActor
        (:Is%20Valid
          (if (not (Utilities|Array|ContainsItem (Variables|Runtime|GetActorsHit) HitActor))
            (Utilities|Array|Add (Variables|Runtime|GetActorsHit) HitActor)
            (Networking|SwitchHasAuthority
              (:Authority
                (Utilities|IsValid _returnvalue_3
                  (:Is%20Valid
                    (bind _damagege (Variables|Default|GetDamageGE)
                      (:then
                        (GameplayEffects|ApplyGameplayEffectToTarget _returnvalue_5 _damagege _returnvalue_3 0.0 _returnvalue_6) ; Apply Damage
                      )
                      (:else
                        _))
                    (bind _knockbackge (Variables|Default|GetKnockbackGE)
                      (:then
                        (GameplayEffects|ApplyGameplayEffectToTarget _returnvalue_5 _knockbackge _returnvalue_3 0.0 _returnvalue_6) ; Apply Knockback
                      )
                      (:else))))))))))))

(event Custom|DestroyBuilding
  (params (BuildingActor :type Building%20Static%20Mesh%20Object%20Reference) (HitComponent :type Primitive%20Component%20Object%20Reference))
  (bind _returnvalue (Components|GetOwner))
  (bind _instigator (Variables|/Script/Engine.Actor|GetInstigator _returnvalue))
  (bind _death_reason (Variables|Default|GetDeath%20Reason))
  ;[0.15 0.03529 0.02561 0.5] Destroy Building: On server call Die to kill the building without needing to replicate a GE. On the client, predict that the actor will be destroyed until the destruction is replicated
  (Utilities|IsValid BuildingActor
    (:Is%20Valid
      (Variables|Runtime|SetBuildingsDestroyed (+ (Variables|Runtime|GetBuildingsDestroyed) (CallFunction|ShouldIncrementBuildingsDestroyed BuildingActor)))
      (Networking|SwitchHasAuthority
        (:Authority
          (Utilities|IsValid _instigator
            (:Is%20Valid
              (Gameplay|Die BuildingActor _death_reason (Pawn|GetController _instigator) _returnvalue :!class /Script/FortniteGame.BuildingActor))
            (:Is%20Not%20Valid
              (Gameplay|Die BuildingActor _death_reason 0 _returnvalue :!class /Script/FortniteGame.BuildingActor))))
        (:Remote
          (bind _asfort_pawn (Utilities|Casting|CastToFortPawn _instigator))
          (Gameplay|PredictDeath BuildingActor _death_reason _asfort_pawn _returnvalue HitComponent false true true))))))

(fn ShouldDestroyBuilding
  (params (BuildingActor :type Building%20Static%20Mesh%20Object%20Reference))
  ;[0.15 0.131 0.04337 0.5] Building should be destroyed if we destroy all buildings or its health is lower than the damage we deal to buildings
  (Utilities|FlowControl|Sequence
    (:then_0
      (Utilities|IsValid BuildingActor
        (:Is%20Valid
          (return (Math|Boolean|ANDBoolean (Variables|/Script/Engine.Actor|GetbCanBeDamaged BuildingActor) (or (< (Variables|Runtime|GetBuildingsDestroyed) (ScalableFloat|GetValueatLevel (Variables|Hotfix|GetHotfix_MaxBuildingsToDestroy))) (<= (ScalableFloat|GetValueatLevel (Variables|Hotfix|GetHotfix_MaxBuildingsToDestroy)) 0)) (or (< (Health|GetHealth BuildingActor :!class /Script/FortniteGame.BuildingActor) (ScalableFloat|GetValueatLevel (Variables|Hotfix|GetHotfix_BuildingDestructionHealthThreshold))) (<= (ScalableFloat|GetValueatLevel (Variables|Hotfix|GetHotfix_BuildingDestructionHealthThreshold)) 0)) (not (GameplayTags|HasAnyMatchingGameplayTags BuildingActor (Variables|Default|GetIgnoreBuildingsWithTags))))))))
    (:then_1
      (return false))))

(fn ShouldIncrementBuildingsDestroyed
  (params (DestroyedActor :type Building%20Actor%20Object%20Reference))
  ;[0.15 0.06583 0.02223 0.5] Don't count low health actors towards max destruction count
  (Utilities|IsValid DestroyedActor
    (:Is%20Valid
      (return (> (Health|GetHealth DestroyedActor :!class /Script/FortniteGame.BuildingActor) (Variables|Hotfix|GetHotfix_BuildingDestructionCountMinHealth))))
    (:Is%20Not%20Valid
      (return false))))
