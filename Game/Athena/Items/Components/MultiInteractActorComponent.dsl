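; Multi-interact actor component that plays local gameplay cues as players join or leave a shared interaction.
; Join and leave fire burst cues on the player's pawn, packing duration progress, current count and previous count into cue parameters.
; Joining also adds a guarded looping cue on the owner.
; It is removed when the count reaches zero or, in local-player-only mode, when the local player leaves, with a count-changed delegate backing up the zero case.
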(event EventBeginPlay
  (EventDispatchers|BindEventtoActorCountChangedDelegate (EventDispatchers|CreateEvent :!event Custom|OnActorCountChanged)))

(event EventUpdateForPlayerJustJoinedInteraction
  (params (JoinedPlayer :type Fort%20Player%20State%20Athena%20Object%20Reference))
  (CallFunction|ExecuteGCNFromTag (PlayerState|GetPlayerPawnorVehicleDriver JoinedPlayer) (Variables|Default|GetAddInteractPlayerTag) (- (CallFunction|GetInteractingActorCount true) 1))
  (CallFunction|AddGCNL))

(event EventUpdateForPlayerJustLeftInteraction
  (params (LeftPlayer :type Fort%20Player%20State%20Athena%20Object%20Reference))
  (bind _returnvalue (PlayerState|GetPlayerPawnorVehicleDriver LeftPlayer))
  (bind _returnvalue_2 (CallFunction|GetInteractingActorCount true))
  (CallFunction|ExecuteGCNFromTag _returnvalue (Variables|Default|GetRemoveInteractPlayerTag) (+ _returnvalue_2 1))
  ;[1 1 1] Check if should remove GCNL
  (if (and (Variables|Default|GetGCNLApplied) (or (and (Pawn|IsLocallyControlled _returnvalue) (Variables|MultiInteract|GetbOnlyUpdateForLocalInteractingPlayer)) (and (not (Variables|MultiInteract|GetbOnlyUpdateForLocalInteractingPlayer)) (== _returnvalue_2 0))))
    (CallFunction|RemoveGCNL)))

(event Custom|OnActorCountChanged
  (params (InteractingActorCount :type Integer) (MultiInteractMultiplier :type Float%20%28single-precision%29) (InteractingDuration :type Float%20%28single-precision%29))
  (if (Math|Boolean|ANDBoolean (== InteractingActorCount 0) (Variables|Default|GetGCNLApplied) (not (Variables|MultiInteract|GetbOnlyUpdateForLocalInteractingPlayer)))
    (CallFunction|RemoveGCNL)))

(event EventEndPlay
  (params (EndPlayReason :type EEndPlayReason%20Enum :default "Destroyed"))
  (EventDispatchers|UnbindEventfromActorCountChangedDelegate (EventDispatchers|CreateEvent :!event Custom|OnActorCountChanged)))

(fn AddGCNL
  (if (not (Variables|Default|GetGCNLApplied))
    (bind _output_get (LocalVariables|SetOwner (Components|GetOwner)))
    (Utilities|IsValid _output_get
      (:Is%20Valid
        (Projectile|AddGameplayCue%28Looping%29Local _output_get (Variables|Default|GetLoopingMultiInteractTag) (Ability|GameplayCue|MakeGameplayCueParameters :NormalizedMagnitude 0.0 :RawMagnitude (Math|Conversions|ToFloat(Integer) (CallFunction|GetInteractingActorCount true)) :Location (Transformation|GetActorLocation (LocalVariables|GetOwner)) :Normal "0,0,0" :Instigator (LocalVariables|GetOwner) :EffectCauser (LocalVariables|GetOwner) :SourceObject (LocalVariables|GetOwner) :GameplayEffectLevel 1 :AbilityLevel 1 :bReplicateLocationWhenUsingMinimalRepProxy false))
        (Variables|Default|SetGCNLApplied true)))))

(fn ExecuteGCNFromTag
  (params (PlayerPawn :type Fort%20Player%20Pawn%20Object%20Reference%20%28by%20ref%29) (InteractTag :type Gameplay%20Tag%20Structure%20%28by%20ref%29) (PreviousInteractPlayerCount :type Integer))
  (Utilities|IsValid PlayerPawn
    (:Is%20Valid
      (GameplayCue|ExecuteGameplayCue%28Burst%29Local (Variables|GetPlayerPawn) (Variables|GetInteractTag) (Ability|GameplayCue|MakeGameplayCueParameters :NormalizedMagnitude (Math|Float|SafeDivide (CallFunction|GetInteractionDuration) (Variables|Default|GetRequiredDuration)) :RawMagnitude (Math|Conversions|ToFloat(Integer) (CallFunction|GetInteractingActorCount true)) :Location (Transformation|GetActorLocation (Variables|GetPlayerPawn)) :Normal "0,0,0" :Instigator (Variables|GetPlayerPawn) :EffectCauser (Components|GetOwner) :GameplayEffectLevel 1 :AbilityLevel (Variables|GetPreviousInteractPlayerCount) :bReplicateLocationWhenUsingMinimalRepProxy false)))))

(fn RemoveGCNL
  (bind _output_get (LocalVariables|SetOwner (Components|GetOwner)))
  (Utilities|IsValid _output_get
    (:Is%20Valid
      (Projectile|RemoveGameplayCue%28Looping%29Local _output_get (Variables|Default|GetLoopingMultiInteractTag) (Ability|GameplayCue|MakeGameplayCueParameters :NormalizedMagnitude 0.0 :RawMagnitude 0.0 :Location "0,0,0" :Normal "0,0,0" :Instigator (LocalVariables|GetOwner) :GameplayEffectLevel 1 :AbilityLevel 1 :bReplicateLocationWhenUsingMinimalRepProxy false))
      (Variables|Default|SetGCNLApplied false))))

(fn SetRequiredDuration
  (params (NewRequiredDuration :type Float%20%28double-precision%29))
  (Variables|Default|SetRequiredDuration NewRequiredDuration))
